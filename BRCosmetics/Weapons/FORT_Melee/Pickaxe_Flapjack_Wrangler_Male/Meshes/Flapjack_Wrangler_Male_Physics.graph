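# Physics asset for the male Flapjack Wrangler two-handed pickaxe, defining jiggle bodies for its pancakes, butter and syrup bones.
# Three box pancakes and a butter box constrain to the handle body, with one pancake nearly rigid and the butter locked on one swing axis.
# Eleven syrup bodies, zero-length capsules acting as spheres, hang from that near-rigid pancake with wide swing and tight twist limits.
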PhysicsAsset /BRCosmetics/Weapons/FORT_Melee/Pickaxe_Flapjack_Wrangler_Male/Meshes/Flapjack_Wrangler_Male_Physics
bodies {
 Dyn_Syrup1 [{shapeType=Capsule, length=0, radius=1.5378, center=<2.0126 -0.5403 -0.0254>, rotation=<rot 0 96.6842 -85.3347>}]
 Dyn_Syrup10 [{shapeType=Capsule, length=0, radius=1.3191, center=<3.3988 -0.1311 0.2094>, rotation=<rot 0 100.4708 -85.8276>}]
 Dyn_Syrup11 [{shapeType=Capsule, length=0, radius=1.3092, center=<3.3968 0.0989 0.1135>, rotation=<rot 0 -76.5942 -89.4588>}]
 Dyn_Syrup2 [{shapeType=Capsule, length=0, radius=1.1354, center=<0.7564 -0.5011 0.1784>, rotation=<rot 0 101.4331 -63.6637>}]
 Dyn_Syrup3 [{shapeType=Capsule, length=0, radius=1.1692, center=<2.5115 -0.0556 -0.105>, rotation=<rot 0 -82.7019 -89.1541>}]
 Dyn_Syrup4 [{shapeType=Capsule, length=0, radius=1.1148, center=<2.5996 -0.6045 -0.2852>, rotation=<rot 0 -84.0207 -87.2181>}]
 Dyn_Syrup5 [{shapeType=Capsule, length=0, radius=1.0634, center=<2.0671 -0.4294 0.2317>, rotation=<rot 0 97.1356 -88.3065>}]
 Dyn_Syrup6 [{shapeType=Capsule, length=0, radius=1.2601, center=<1.5944 0.2839 -0.1761>, rotation=<rot 0 -89.007 -79.329>}]
 Dyn_Syrup7 [{shapeType=Capsule, length=0, radius=1.8283, center=<4.0656 -0.3509 -0.1423>, rotation=<rot 0 -86.9094 -83.115>}]
 Dyn_Syrup8 [{shapeType=Capsule, length=0, radius=1.2114, center=<3.1597 0.3503 -0.0293>, rotation=<rot 0 93.6843 -84.5903>}]
 Dyn_Syrup9 [{shapeType=Capsule, length=0, radius=1.0914, center=<3.2674 -0.562 0>}]
 Dyn_pancake1 [{shapeType=Box, extentX=6.6808, extentY=2, extentZ=20.9759, center=<3.9186 0 0>}]
 Dyn_pancake2 [{shapeType=Box, extentX=12.3821, extentY=2, extentZ=20.9759, center=<4.3291 0 0>}]
 Dyn_pancake3 [{shapeType=Box, extentX=22.3373, extentY=2, extentZ=22.3373, center=<0 0.925 0>}]
 Melee_TwoHanded_Handle [{shapeType=Capsule, length=0, radius=13.8345, center=<0.3473 -1.9115 68.8802>, rotation=<rot 0 90 0>}]
 dyn_butter [{shapeType=Box, extentX=2, extentY=2, extentZ=2, center=<0 0 -0.5793>}]
}
constraints [
 {bone1Name=Dyn_pancake3, bone2Name=Melee_TwoHanded_Handle, swing1LimitDegrees=1, swing2LimitDegrees=1, twistLimitDegrees=1}
 {bone1Name=Dyn_pancake1, bone2Name=Melee_TwoHanded_Handle, swing1LimitDegrees=20, swing2LimitDegrees=10, twistLimitDegrees=5}
 {bone1Name=Dyn_pancake2, bone2Name=Melee_TwoHanded_Handle, swing1LimitDegrees=20, swing2LimitDegrees=10, twistLimitDegrees=5}
 {bone1Name=Dyn_Syrup1, bone2Name=Dyn_pancake3, swing1LimitDegrees=65, swing2LimitDegrees=65, twistLimitDegrees=5}
 {bone1Name=Dyn_Syrup2, bone2Name=Dyn_pancake3, swing1LimitDegrees=58, swing2LimitDegrees=58, twistLimitDegrees=5}
 {bone1Name=Dyn_Syrup3, bone2Name=Dyn_pancake3, swing1LimitDegrees=65, swing2LimitDegrees=65, twistLimitDegrees=5}
 {bone1Name=Dyn_Syrup4, bone2Name=Dyn_pancake3, swing1LimitDegrees=65, swing2LimitDegrees=65, twistLimitDegrees=5}
 {bone1Name=Dyn_Syrup5, bone2Name=Dyn_pancake3, swing1LimitDegrees=45, swing2LimitDegrees=45, twistLimitDegrees=5}
 {bone1Name=Dyn_Syrup6, bone2Name=Dyn_pancake3, swing1LimitDegrees=45, swing2LimitDegrees=45, twistLimitDegrees=5}
 {bone1Name=Dyn_Syrup7, bone2Name=Dyn_pancake3, swing1LimitDegrees=45, swing2LimitDegrees=45, twistLimitDegrees=5}
 {bone1Name=Dyn_Syrup8, bone2Name=Dyn_pancake3, swing1LimitDegrees=45, swing2LimitDegrees=45, twistLimitDegrees=5}
 {bone1Name=Dyn_Syrup9, bone2Name=Dyn_pancake3, swing1LimitDegrees=40, swing2LimitDegrees=40, twistLimitDegrees=8}
 {bone1Name=Dyn_Syrup10, bone2Name=Dyn_pancake3, swing1LimitDegrees=55, swing2LimitDegrees=55, twistLimitDegrees=5}
 {bone1Name=Dyn_Syrup11, bone2Name=Dyn_pancake3, swing1LimitDegrees=65, swing2LimitDegrees=65, twistLimitDegrees=5}
 {
  bone1Name=dyn_butter
  bone2Name=Melee_TwoHanded_Handle
  swing1LimitDegrees=45
  swing1Motion=Locked
  swing2LimitDegrees=5
  twistLimitDegrees=5
 }
]
