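# Animation Blueprint for the Flapjack Wrangler Male pickaxe that adds secondary physics motion to the weapon mesh.
# The output pose comes from a RigidBody node running on the local ref pose with an override physics asset, simulated relative to b_AR_Root with overridden gravity.
# Anim Dynamics nodes for three chained pancake bones and a separate butter bone remain in the graph but are not connected to the output.
# The EventGraph holds only unconnected default nodes.
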AnimBlueprint /BRCosmetics/Weapons/FORT_Melee/Pickaxe_Flapjack_Wrangler_Male/Meshes/Flapjack_Wrangler_Male_AnimBP
target_skeleton=&/Flapjack_Wrangler_Male_Skeleton
graphs [
 AnimGraph {nodes [
   n1 = Anim Dynamics - Bone: Dyn_pancake1(Alpha=1, ExternalForce="0,0,0") {
    bGravityOverrideInSimSpace=true
    simulationSpace=BoneRelative
    alphaBoolBlend {blendOption=Linear}
    boundBone {boneName=Dyn_pancake1}
    componentLinearAccScale=<0.25 0.25 0.4>
    physicsBodyDefinitions [{
      sphereCollisionRadius=0
      boundBone {boneName=Dyn_pancake1}
      boxExtents=<5 5 5>
      constraintSetup {angularLimitsMax=<0 0 15>, angularLimitsMin=<0 0 -15>, angularTarget=<0 0 0>}
      localJointOffset=<0 0 -1>
     }]
    relativeSpaceBone {boneName=b_AR_Root}
   }
   n2 = Anim Dynamics - Bone: Dyn_pancake2(ComponentPose=@n1, Alpha=1, ExternalForce="0,0,0") {
    bGravityOverrideInSimSpace=true
    simulationSpace=BoneRelative
    alphaBoolBlend {blendOption=Linear}
    boundBone {boneName=Dyn_pancake2}
    componentLinearAccScale=<0.25 0.25 0.4>
    physicsBodyDefinitions [{
      sphereCollisionRadius=0
      boundBone {boneName=Dyn_pancake2}
      boxExtents=<5 5 5>
      constraintSetup {angularLimitsMax=<0 0 15>, angularLimitsMin=<0 0 -15>, angularTarget=<0 0 0>}
      localJointOffset=<0 0 -1>
     }]
    relativeSpaceBone {boneName=b_AR_Root}
   }
   Anim Dynamics - Bone: Dyn_pancake3(ComponentPose=@n2, Alpha=1, ExternalForce="0,0,0") {
    bGravityOverrideInSimSpace=true
    simulationSpace=BoneRelative
    alphaBoolBlend {blendOption=Linear}
    boundBone {boneName=Dyn_pancake3}
    componentLinearAccScale=<0.25 0.25 0.4>
    physicsBodyDefinitions [{
      sphereCollisionRadius=0
      boundBone {boneName=Dyn_pancake3}
      boxExtents=<5 5 5>
      constraintSetup {angularLimitsMax=<0 0 15>, angularLimitsMin=<0 0 -15>, angularTarget=<0 0 0>}
      localJointOffset=<0 0 -1>
     }]
    relativeSpaceBone {boneName=b_AR_Root}
   }
   Anim Dynamics - Bone: dyn_butter(Alpha=1, ExternalForce="0,0,0") {
    bGravityOverrideInSimSpace=true
    alphaBoolBlend {blendOption=Linear}
    boundBone {boneName=dyn_butter}
    componentLinearAccScale=<0.4 0.4 0.5>
    physicsBodyDefinitions [{
      sphereCollisionRadius=0
      boundBone {boneName=dyn_butter}
      boxExtents=<5 5 5>
      constraintSetup {angularLimitsMax=<0 0 0>, angularLimitsMin=<0 0 0>, angularTarget=<0 0 0>, linearAxesMax=<1 1 1>, linearAxesMin=<-1 -1 -1>}
      localJointOffset=<0 0 -1>
     }]
    relativeSpaceBone {boneName=b_AR_Root}
   }
   n3 = RigidBody(ComponentPose=Local To Component(LocalPose=Local Space Ref Pose())) {
    bClampLinearTranslationLimitToRefPose=true
    bOverrideWorldGravity=true
    overridePhysicsAsset=&/Flapjack_Wrangler_Male_Physics
    simulationSpace=BaseBoneSpace
    alphaBoolBlend {blendOption=Linear}
    baseBoneRef {boneName=b_AR_Root}
    componentAppliedLinearAccClamp=<1000 1000 1000>
    componentLinearAccScale=<0.4 0.4 0.7>
    overrideWorldGravity {z=-980}
   }
   Output Pose(Result=Component To Local(ComponentPose=@n3)) {layerGroup=None}
  ]}
 EventGraph {nodes [
   TryGetPawnOwner()
   Event BlueprintUpdateAnimation()
  ]}
]
variables []
